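# Physics asset for the dynamic bones on the large male Broccoli Green player body, covering the belly, both lats and both love handles.
# Each dynamic bone has a capsule body, and smaller capsules on spine_01 and spine_04 serve as their attachment points.
# The lats are constrained to spine_04, while the belly and love handles are constrained to spine_01.
# The constraint between the two spine bodies leaves swing and twist free.
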PhysicsAsset /BRCosmetics/Characters/Player/Male/Large/Bodies/M_LRG_Broccoli_Green/Meshes/M_LRG_Broccoli_Green_Dyn
bodies {
 dyn_belly [{shapeType=Capsule, length=5, radius=5}]
 dyn_lat_l_01 [{shapeType=Capsule, length=5, radius=5}]
 dyn_lat_r_01 [{shapeType=Capsule, length=5, radius=5}]
 dyn_lovehandle_l_01 [{shapeType=Capsule, length=5, radius=5}]
 dyn_rovehandle_r_01 [{shapeType=Capsule, length=5, radius=5}]
 spine_01 [{shapeType=Capsule, length=1, radius=1}]
 spine_04 [{shapeType=Capsule, length=1, radius=1}]
}
constraints [
 {bone1Name=dyn_lat_r_01, bone2Name=spine_04, swing1LimitDegrees=45, swing2LimitDegrees=45, twistLimitDegrees=45}
 {bone1Name=dyn_lat_l_01, bone2Name=spine_04, swing1LimitDegrees=45, swing2LimitDegrees=45, twistLimitDegrees=45}
 {
  bone1Name=spine_04
  bone2Name=spine_01
  swing1LimitDegrees=45
  swing1Motion=Free
  swing2LimitDegrees=45
  swing2Motion=Free
  twistLimitDegrees=45
  twistMotion=Free
 }
 {bone1Name=dyn_belly, bone2Name=spine_01, swing1LimitDegrees=45, swing2LimitDegrees=45, twistLimitDegrees=45}
 {bone1Name=dyn_rovehandle_r_01, bone2Name=spine_01, swing1LimitDegrees=45, swing2LimitDegrees=45, twistLimitDegrees=45}
 {bone1Name=dyn_lovehandle_l_01, bone2Name=spine_01, swing1LimitDegrees=45, swing2LimitDegrees=45, twistLimitDegrees=45}
]
