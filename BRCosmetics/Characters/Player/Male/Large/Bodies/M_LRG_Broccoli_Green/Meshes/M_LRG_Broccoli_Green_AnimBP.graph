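# Post-processes the pose for the large male Broccoli Green player body cosmetic, adding deformation and secondary motion.
# Copies the attached parent's pose, layers the HeadMesh pose from C_jaw down, then runs twist and sculpter Control Rigs.
# A RigidBody simulation based at spine_04 feeds weighted multi-bone constraints that position arm, hand, torso and leg fur bones.
# Chained per-bone AnimDynamics nodes then add damped, angle-limited spring jiggle to those bones.
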AnimBlueprint /BRCosmetics/Characters/Player/Male/Large/Bodies/M_LRG_Broccoli_Green/Meshes/M_LRG_Broccoli_Green_AnimBP
target_skeleton=&/M_LRG_Broccoli_Green_Skeleton
graphs [
 AnimGraph {nodes [
   n1 = Get HeadMesh() -> (HeadMesh:Skeletal Mesh Component Object Reference)
   n2 = Control Rig(Source=Layered blend per bone(BasePose=Copy Pose From Mesh() {bUseAttachedParent=true}, BlendPoses_0=Copy Pose From Mesh(SourceMeshComponent=@n1) {bCopyCurves=true}) {layerSetup [{branchFilters [{boneName=C_jaw}]}]}) {
    alphaBoolBlend {blendOption=Linear}
    controlRigAssetReference {blueprintRigClass=&/M_LRG_Broccoli_Green_Deform_CtrlRig_twists.M_LRG_Broccoli_Green_Deform_CtrlRig_twists_C}
   }
   n3 = Control Rig!(
    Source=@n2
    Alpha
    KneeRot_fr_l_114deg:Vector
    KneeTrans_fr_l_114deg:Vector="0.258,-3.591,0.009"
    KneeScale_fr_l_114deg:Vector="1,1,1"
    KneeScale_fr_l_130deg:Vector="1,1,1"
    KneeRot_fr_l_130deg:Vector="0,0,-51.2"
    KneeTrans_fr_l_130deg:Vector="0.258,-3.591,0.009"
   ) -> (Pose) {
    alphaBoolBlend {blendOption=Linear}
    controlRigAssetReference {blueprintRigClass=&/M_LRG_Broccoli_Green_Deform_CtrlRig_sculpters.M_LRG_Broccoli_Green_Deform_CtrlRig_sculpters_C}
   }
   n4 = RigidBody!(ComponentPose=Local To Component(LocalPose=@n3), Alpha=0.8, OverrideWorldGravity:Vector="0,0,-980", ExternalForce) -> (Pose) {
    bClampLinearTranslationLimitToRefPose=true
    bOverrideWorldGravity=true
    overridePhysicsAsset=&/M_LRG_Broccoli_Green_Dyn
    simulationSpace=BaseBoneSpace
    alphaBoolBlend {blendOption=Linear}
    baseBoneRef {boneName=spine_04}
    componentAppliedLinearAccClamp=<3000 3000 3000>
    componentLinearAccScale {z=0.8}
    componentLinearVelScale {z=0.8}
    overrideWorldGravity {z=-980}
    simSpaceSettings {
     maxAngularAcceleration=500
     maxAngularVelocity=3
     maxLinearAcceleration=700
     maxLinearVelocity=500
     velocityScaleZ=0
     worldAlpha=1
     externalLinearDragV=<1 1>
    }
   }
   n5 = "Constraint - dyn_fur_arm_r_01 (PPPPPP)"!(
    ComponentPose=@n4
    Alpha=1
    ConstraintWeights_0=0.3088
    ConstraintWeights_1:float=0.2881
    ConstraintWeights_2:float=0.2643
    ConstraintWeights_3:float=0.0958
    ConstraintWeights_4:float=0.0226
    ConstraintWeights_5:float=0.0205
   ) -> (Pose) {
    alphaBoolBlend {blendOption=Linear}
    boneToModify {boneName=dyn_fur_arm_r_01}
    constraintSetup [
     {targetBone {boneName=dyn_fur_arm_r_01}}
     {targetBone {boneName=lowerarm_r}}
     {targetBone {boneName=upperarm_r}}
     {targetBone {boneName=lowerarm_twist_02_r}}
     {targetBone {boneName=upperarm_twist_02_r}}
     {targetBone {boneName=deform_elbow_fr_r}}
    ]
    constraintWeights [1,1,1,1,1,1]
   }
   n6 = "Constraint - dyn_fur_arm_r_02 (PPPPPP)"!(
    ComponentPose=@n5
    Alpha=1
    ConstraintWeights_0=0.4659
    ConstraintWeights_1:float=0.3282
    ConstraintWeights_2:float=0.0774
    ConstraintWeights_3:float=0.0736
    ConstraintWeights_4:float=0.0293
    ConstraintWeights_5:float=0.0256
   ) -> (Pose) {
    alphaBoolBlend {blendOption=Linear}
    boneToModify {boneName=dyn_fur_arm_r_02}
    constraintSetup [
     {targetBone {boneName=dyn_fur_arm_r_02}}
     {targetBone {boneName=lowerarm_r}}
     {targetBone {boneName=lowerarm_twist_02_r}}
     {targetBone {boneName=upperarm_r}}
     {targetBone {boneName=lowerarm_twist_01_r}}
     {targetBone {boneName=deform_elbow_bk_r}}
    ]
    constraintWeights [1,1,1,1,1,1]
   }
   n7 = "Constraint - dyn_fur_arm_r_03 (PPPP)"!(
    ComponentPose=@n6
    Alpha=1
    ConstraintWeights_0=0.3811
    ConstraintWeights_1:float=0.2936
    ConstraintWeights_2:float=0.2183
    ConstraintWeights_3:float=0.107
   ) -> (Pose) {
    alphaBoolBlend {blendOption=Linear}
    boneToModify {boneName=dyn_fur_arm_r_03}
    constraintSetup [
     {targetBone {boneName=upperarm_r}}
     {targetBone {boneName=upperarm_twist_02_r}}
     {targetBone {boneName=upperarm_twist_01_r}}
     {targetBone {boneName=dyn_fur_arm_r_03}}
    ]
    constraintWeights [1,1,1,1]
   }
   n8 = "Constraint - dyn_fur_arm_r_04 (PPPPPPPP)"!(
    ComponentPose=@n7
    Alpha=1
    ConstraintWeights_0=0.2866
    ConstraintWeights_1:float=0.2418
    ConstraintWeights_2:float=0.201
    ConstraintWeights_3:float=0.1013
    ConstraintWeights_4:float=0.0654
    ConstraintWeights_5:float=0.0562
    ConstraintWeights_6:float=0.0346
    ConstraintWeights_7:float=0.0131
   ) -> (Pose) {
    alphaBoolBlend {blendOption=Linear}
    boneToModify {boneName=dyn_fur_arm_r_04}
    constraintSetup [
     {targetBone {boneName=upperarm_twist_01_r}}
     {targetBone {boneName=dyn_fur_arm_r_04}}
     {targetBone {boneName=upperarm_twist_02_r}}
     {targetBone {boneName=deform_deltoid_fr_r}}
     {targetBone {boneName=clavicle_r}}
     {targetBone {boneName=deform_pectoral_fr_r}}
     {targetBone {boneName=upperarm_r}}
     {targetBone {boneName=spine_03}}
    ]
    constraintWeights [1,1,1,1,1,1,1,1]
   }
   n9 = "Constraint - dyn_fur_hand_r_05 (PPPP)"!(
    ComponentPose=@n8
    Alpha=1
    ConstraintWeights_0=0.6131
    ConstraintWeights_1:float=0.2274
    ConstraintWeights_2:float=0.0953
    ConstraintWeights_3:float=0.0642
   ) -> (Pose) {
    alphaBoolBlend {blendOption=Linear}
    boneToModify {boneName=dyn_fur_hand_r_05}
    constraintSetup [
     {targetBone {boneName=hand_r}}
     {targetBone {boneName=dyn_fur_hand_r_05}}
     {targetBone {boneName=dyn_fur_hand_r_06}}
     {targetBone {boneName=lowerarm_twist_01_r}}
    ]
    constraintWeights [1,1,1,1]
   }
   n10 = "Constraint - dyn_fur_hand_r_06 (PPPP)"!(
    ComponentPose=@n9
    Alpha=1
    ConstraintWeights_0=0.6131
    ConstraintWeights_1:float=0.2274
    ConstraintWeights_2:float=0.0953
    ConstraintWeights_3:float=0.0642
   ) -> (Pose) {
    alphaBoolBlend {blendOption=Linear}
    boneToModify {boneName=dyn_fur_hand_r_06}
    constraintSetup [
     {targetBone {boneName=hand_r}}
     {targetBone {boneName=dyn_fur_hand_r_05}}
     {targetBone {boneName=dyn_fur_hand_r_06}}
     {targetBone {boneName=lowerarm_twist_01_r}}
    ]
    constraintWeights [1,1,1,1]
   }
   n11 = "Constraint - dyn_fur_arm_l_01 (PPPPP)"!(
    ComponentPose=@n10
    Alpha=1
    ConstraintWeights_0=0.5034
    ConstraintWeights_1:float=0.2386
    ConstraintWeights_2:float=0.1789
    ConstraintWeights_3:float=0.0561
    ConstraintWeights_4:float=0.023
   ) -> (Pose) {
    alphaBoolBlend {blendOption=Linear}
    boneToModify {boneName=dyn_fur_arm_l_01}
    constraintSetup [
     {targetBone {boneName=dyn_fur_arm_l_01}}
     {targetBone {boneName=lowerarm_l}}
     {targetBone {boneName=upperarm_l}}
     {targetBone {boneName=lowerarm_twist_02_l}}
     {targetBone {boneName=deform_elbow_bk_l}}
    ]
    constraintWeights [1,1,1,1,1]
   }
   n12 = "Constraint - dyn_fur_arm_l_02 (PPPPPP)"!(
    ComponentPose=@n11
    Alpha=1
    ConstraintWeights_0=0.4659
    ConstraintWeights_1:float=0.3282
    ConstraintWeights_2:float=0.0774
    ConstraintWeights_3:float=0.0736
    ConstraintWeights_4:float=0.0293
    ConstraintWeights_5:float=0.0256
   ) -> (Pose) {
    alphaBoolBlend {blendOption=Linear}
    boneToModify {boneName=dyn_fur_arm_l_02}
    constraintSetup [
     {targetBone {boneName=dyn_fur_arm_l_02}}
     {targetBone {boneName=lowerarm_l}}
     {targetBone {boneName=lowerarm_twist_02_l}}
     {targetBone {boneName=upperarm_l}}
     {targetBone {boneName=lowerarm_twist_01_l}}
     {targetBone {boneName=deform_elbow_bk_l}}
    ]
    constraintWeights [1,1,1,1,1,1]
   }
   n13 = "Constraint - dyn_fur_arm_l_03 (PPPP)"!(
    ComponentPose=@n12
    Alpha=1
    ConstraintWeights_0=0.3811
    ConstraintWeights_1:float=0.2936
    ConstraintWeights_2:float=0.2183
    ConstraintWeights_3:float=0.107
   ) -> (Pose) {
    alphaBoolBlend {blendOption=Linear}
    boneToModify {boneName=dyn_fur_arm_l_03}
    constraintSetup [
     {targetBone {boneName=upperarm_l}}
     {targetBone {boneName=upperarm_twist_02_l}}
     {targetBone {boneName=upperarm_twist_01_l}}
     {targetBone {boneName=dyn_fur_arm_l_03}}
    ]
    constraintWeights [1,1,1,1]
   }
   n14 = "Constraint - dyn_fur_arm_l_04 (PPPPPPPP)"!(
    ComponentPose=@n13
    Alpha=1
    ConstraintWeights_0=0.2866
    ConstraintWeights_1:float=0.2418
    ConstraintWeights_2:float=0.201
    ConstraintWeights_3:float=0.1013
    ConstraintWeights_4:float=0.0654
    ConstraintWeights_5:float=0.0562
    ConstraintWeights_6:float=0.0346
    ConstraintWeights_7:float=0.013
   ) -> (Pose) {
    alphaBoolBlend {blendOption=Linear}
    boneToModify {boneName=dyn_fur_arm_l_04}
    constraintSetup [
     {targetBone {boneName=upperarm_twist_01_l}}
     {targetBone {boneName=dyn_fur_arm_l_04}}
     {targetBone {boneName=upperarm_twist_02_l}}
     {targetBone {boneName=deform_deltoid_fr_l}}
     {targetBone {boneName=clavicle_l}}
     {targetBone {boneName=deform_pectoral_fr_l}}
     {targetBone {boneName=upperarm_l}}
     {targetBone {boneName=spine_03}}
    ]
    constraintWeights [1,1,1,1,1,1,1,1]
   }
   n15 = "Constraint - dyn_fur_hand_l_05 (PPPP)"!(
    ComponentPose=@n14
    Alpha=1
    ConstraintWeights_0=0.6131
    ConstraintWeights_1:float=0.2274
    ConstraintWeights_2:float=0.0953
    ConstraintWeights_3:float=0.0642
   ) -> (Pose) {
    alphaBoolBlend {blendOption=Linear}
    boneToModify {boneName=dyn_fur_hand_l_05}
    constraintSetup [
     {targetBone {boneName=hand_l}}
     {targetBone {boneName=dyn_fur_hand_l_05}}
     {targetBone {boneName=dyn_fur_hand_l_06}}
     {targetBone {boneName=lowerarm_twist_01_l}}
    ]
    constraintWeights [1,1,1,1]
   }
   n16 = "Constraint - dyn_fur_hand_l_06 (PPPP)"!(
    ComponentPose=@n15
    Alpha=1
    ConstraintWeights_0=0.6131
    ConstraintWeights_1:float=0.2274
    ConstraintWeights_2:float=0.0953
    ConstraintWeights_3:float=0.0642
   ) -> (Pose) {
    alphaBoolBlend {blendOption=Linear}
    boneToModify {boneName=dyn_fur_hand_l_06}
    constraintSetup [
     {targetBone {boneName=hand_l}}
     {targetBone {boneName=dyn_fur_hand_l_05}}
     {targetBone {boneName=dyn_fur_hand_l_06}}
     {targetBone {boneName=lowerarm_twist_01_l}}
    ]
    constraintWeights [1,1,1,1]
   }
   n17 = "Constraint - dyn_fur_torso_r_01 (PPPPPP)"!(
    ComponentPose=@n16
    Alpha=1
    ConstraintWeights_0=0.4969
    ConstraintWeights_1:float=0.4098
    ConstraintWeights_2:float=0.0359
    ConstraintWeights_3:float=0.03
    ConstraintWeights_4:float=0.0154
    ConstraintWeights_5:float=0.0119
   ) -> (Pose) {
    alphaBoolBlend {blendOption=Linear}
    boneToModify {boneName=dyn_fur_torso_r_01}
    constraintSetup [
     {targetBone {boneName=dyn_fur_torso_r_01}}
     {targetBone {boneName=pelvis}}
     {targetBone {boneName=thigh_twist_01_r}}
     {targetBone {boneName=spine_01}}
     {targetBone {boneName=spine_02}}
     {targetBone {boneName=spine_03}}
    ]
    constraintWeights [1,1,1,1,1,1]
   }
   n18 = "Constraint - dyn_fur_torso_r_02 (PPPPP)"!(
    ComponentPose=@n17
    Alpha=1
    ConstraintWeights_0=0.4384
    ConstraintWeights_1:float=0.3486
    ConstraintWeights_2:float=0.1749
    ConstraintWeights_3:float=0.027
    ConstraintWeights_4:float=0.011
   ) -> (Pose) {
    alphaBoolBlend {blendOption=Linear}
    boneToModify {boneName=dyn_fur_torso_r_02}
    constraintSetup [
     {targetBone {boneName=thigh_twist_01_r}}
     {targetBone {boneName=dyn_fur_torso_r_02}}
     {targetBone {boneName=pelvis}}
     {targetBone {boneName=thigh_r}}
     {targetBone {boneName=deform_thigh_bk_r}}
    ]
    constraintWeights [1,1,1,1,1]
   }
   n19 = "Constraint - dyn_fur_torso_l_01 (PPPPPP)"!(
    ComponentPose=@n18
    Alpha=1
    ConstraintWeights_0=0.4753
    ConstraintWeights_1:float=0.4275
    ConstraintWeights_2:float=0.0375
    ConstraintWeights_3:float=0.0313
    ConstraintWeights_4:float=0.016
    ConstraintWeights_5:float=0.0124
   ) -> (Pose) {
    alphaBoolBlend {blendOption=Linear}
    boneToModify {boneName=dyn_fur_torso_l_01}
    constraintSetup [
     {targetBone {boneName=dyn_fur_torso_l_01}}
     {targetBone {boneName=pelvis}}
     {targetBone {boneName=thigh_twist_01_l}}
     {targetBone {boneName=spine_01}}
     {targetBone {boneName=spine_02}}
     {targetBone {boneName=spine_03}}
    ]
    constraintWeights [1,1,1,1,1,1]
   }
   n20 = "Constraint - dyn_fur_torso_l_02 (PPPP)"!(
    ComponentPose=@n19
    Alpha=1
    ConstraintWeights_0=0.4203
    ConstraintWeights_1:float=0.4096
    ConstraintWeights_2:float=0.1417
    ConstraintWeights_3:float=0.0284
   ) -> (Pose) {
    alphaBoolBlend {blendOption=Linear}
    boneToModify {boneName=dyn_fur_torso_l_02}
    constraintSetup [
     {targetBone {boneName=dyn_fur_torso_l_02}}
     {targetBone {boneName=thigh_twist_01_l}}
     {targetBone {boneName=pelvis}}
     {targetBone {boneName=thigh_l}}
    ]
    constraintWeights [1,1,1,1]
   }
   n21 = "Constraint - dyn_fur_leg_r_01 (PPP)"!(ComponentPose=@n20, Alpha=1, ConstraintWeights_0=0.4003, ConstraintWeights_1:float=0.3006, ConstraintWeights_2:float=0.2991) -> (Pose) {
    alphaBoolBlend {blendOption=Linear}
    boneToModify {boneName=dyn_fur_leg_r_01}
    constraintSetup [{targetBone {boneName=thigh_twist_01_r}}, {targetBone {boneName=thigh_r}}, {targetBone {boneName=dyn_fur_leg_r_01}}]
    constraintWeights [1,1,1]
   }
   n22 = "Constraint - dyn_fur_leg_r_02 (PPPP)"!(
    ComponentPose=@n21
    Alpha=1
    ConstraintWeights_0=0.3993
    ConstraintWeights_1:float=0.2625
    ConstraintWeights_2:float=0.2066
    ConstraintWeights_3:float=0.1316
   ) -> (Pose) {
    alphaBoolBlend {blendOption=Linear}
    boneToModify {boneName=dyn_fur_leg_r_02}
    constraintSetup [
     {targetBone {boneName=thigh_r}}
     {targetBone {boneName=dyn_fur_leg_r_02}}
     {targetBone {boneName=calf_r}}
     {targetBone {boneName=deform_knee_fr_01_r}}
    ]
    constraintWeights [1,1,1,1]
   }
   n23 = "Constraint - dyn_fur_leg_r_03 (PPPP)"!(
    ComponentPose=@n22
    Alpha=1
    ConstraintWeights_0=0.5114
    ConstraintWeights_1:float=0.2507
    ConstraintWeights_2:float=0.1245
    ConstraintWeights_3:float=0.1134
   ) -> (Pose) {
    alphaBoolBlend {blendOption=Linear}
    boneToModify {boneName=dyn_fur_leg_r_03}
    constraintSetup [
     {targetBone {boneName=dyn_fur_leg_r_03}}
     {targetBone {boneName=calf_twist_02_r}}
     {targetBone {boneName=calf_r}}
     {targetBone {boneName=calf_twist_01_r}}
    ]
    constraintWeights [1,1,1,1]
   }
   n24 = "Constraint - dyn_fur_leg_r_04 (PPP)"!(ComponentPose=@n23, Alpha=1, ConstraintWeights_0=0.7584, ConstraintWeights_1:float=0.1927, ConstraintWeights_2:float=0.0489) -> (Pose) {
    alphaBoolBlend {blendOption=Linear}
    boneToModify {boneName=dyn_fur_leg_r_04}
    constraintSetup [{targetBone {boneName=foot_r}}, {targetBone {boneName=dyn_fur_leg_r_04}}, {targetBone {boneName=calf_twist_01_r}}]
    constraintWeights [1,1,1]
   }
   n25 = "Constraint - dyn_fur_leg_r_05 (PPP)"!(ComponentPose=@n24, Alpha=1, ConstraintWeights_0=0.8673, ConstraintWeights_1:float=0.1218, ConstraintWeights_2:float=0.0109) -> (Pose) {
    alphaBoolBlend {blendOption=Linear}
    boneToModify {boneName=dyn_fur_leg_r_05}
    constraintSetup [{targetBone {boneName=foot_r}}, {targetBone {boneName=dyn_fur_leg_r_05}}, {targetBone {boneName=calf_twist_01_r}}]
    constraintWeights [1,1,1]
   }
   n26 = "Constraint - dyn_fur_leg_l_01 (PPP)"!(ComponentPose=@n25, Alpha=1, ConstraintWeights_0=0.3721, ConstraintWeights_1:float=0.3462, ConstraintWeights_2:float=0.2817) -> (Pose) {
    alphaBoolBlend {blendOption=Linear}
    boneToModify {boneName=dyn_fur_leg_l_01}
    constraintSetup [{targetBone {boneName=dyn_fur_leg_l_01}}, {targetBone {boneName=thigh_twist_01_l}}, {targetBone {boneName=thigh_l}}]
    constraintWeights [1,1,1]
   }
   n27 = "Constraint - dyn_fur_leg_l_02 (PPPP)"!(
    ComponentPose=@n26
    Alpha=1
    ConstraintWeights_0=0.3992
    ConstraintWeights_1:float=0.2626
    ConstraintWeights_2:float=0.2066
    ConstraintWeights_3:float=0.1316
   ) -> (Pose) {
    alphaBoolBlend {blendOption=Linear}
    boneToModify {boneName=dyn_fur_leg_l_02}
    constraintSetup [
     {targetBone {boneName=thigh_l}}
     {targetBone {boneName=dyn_fur_leg_l_02}}
     {targetBone {boneName=calf_l}}
     {targetBone {boneName=deform_knee_fr_01_l}}
    ]
    constraintWeights [1,1,1,1]
   }
   n28 = "Constraint - dyn_fur_leg_l_03 (PPPP)"!(
    ComponentPose=@n27
    Alpha=1
    ConstraintWeights_0=0.5114
    ConstraintWeights_1:float=0.2507
    ConstraintWeights_2:float=0.1245
    ConstraintWeights_3:float=0.1134
   ) -> (Pose) {
    alphaBoolBlend {blendOption=Linear}
    boneToModify {boneName=dyn_fur_leg_l_03}
    constraintSetup [
     {targetBone {boneName=dyn_fur_leg_l_03}}
     {targetBone {boneName=calf_twist_02_l}}
     {targetBone {boneName=calf_l}}
     {targetBone {boneName=calf_twist_01_l}}
    ]
    constraintWeights [1,1,1,1]
   }
   n29 = "Constraint - dyn_fur_leg_l_04 (PPP)"!(ComponentPose=@n28, Alpha=1, ConstraintWeights_0=0.7584, ConstraintWeights_1:float=0.1927, ConstraintWeights_2:float=0.0489) -> (Pose) {
    alphaBoolBlend {blendOption=Linear}
    boneToModify {boneName=dyn_fur_leg_l_04}
    constraintSetup [{targetBone {boneName=foot_l}}, {targetBone {boneName=dyn_fur_leg_l_04}}, {targetBone {boneName=calf_twist_01_l}}]
    constraintWeights [1,1,1]
   }
   n30 = "Constraint - dyn_fur_leg_l_05 (PPP)"!(ComponentPose=@n29, Alpha=1, ConstraintWeights_0=0.8673, ConstraintWeights_1:float=0.1218, ConstraintWeights_2:float=0.0109) -> (Pose) {
    alphaBoolBlend {blendOption=Linear}
    boneToModify {boneName=dyn_fur_leg_l_05}
    constraintSetup [{targetBone {boneName=foot_l}}, {targetBone {boneName=dyn_fur_leg_l_05}}, {targetBone {boneName=calf_twist_01_l}}]
    constraintWeights [1,1,1]
   }
   n31 = Anim Dynamics - Bone: dyn_fur_arm_r_01(ComponentPose=@n30, Alpha=1, ExternalForce="0,0,0") {
    angularDampingOverride=0.9999
    angularSpringConstant=20000
    bAngularSpring=true
    bGravityOverrideInSimSpace=true
    bOverrideAngularDamping=true
    bUseGravityOverride=true
    gravityScale=0.35
    simulationSpace=RootRelative
    alphaBoolBlend {blendOption=Linear}
    boundBone {boneName=dyn_fur_arm_r_01}
    componentAppliedLinearAccClamp=<3000 3000 3000>
    componentLinearAccScale=<0.4 0.4 0.4>
    componentLinearVelScale=<0.4 0.4 0.4>
    gravityOverride {z=-980}
    physicsBodyDefinitions [{
      sphereCollisionRadius=0
      boundBone {boneName=dyn_fur_arm_r_01}
      boxExtents=<15 18 26>
      constraintSetup {angularLimitsMax=<15 15 15>, angularLimitsMin=<-15 -15 -15>, angularTarget=<1 0 0>}
      localJointOffset=<4.5 0 0>
     }]
   }
   n32 = Anim Dynamics - Bone: dyn_fur_arm_r_02(ComponentPose=@n31, Alpha=1, ExternalForce="0,0,0") {
    angularDampingOverride=0.9999
    angularSpringConstant=20000
    bAngularSpring=true
    bGravityOverrideInSimSpace=true
    bOverrideAngularDamping=true
    bUseGravityOverride=true
    gravityScale=0.35
    simulationSpace=RootRelative
    alphaBoolBlend {blendOption=Linear}
    boundBone {boneName=dyn_fur_arm_r_02}
    componentAppliedLinearAccClamp=<3000 3000 3000>
    componentLinearAccScale=<0.4 0.4 0.4>
    componentLinearVelScale=<0.4 0.4 0.4>
    gravityOverride {z=-980}
    physicsBodyDefinitions [{
      sphereCollisionRadius=0
      boundBone {boneName=dyn_fur_arm_r_02}
      boxExtents=<10 20 20>
      constraintSetup {angularLimitsMax=<15 15 15>, angularLimitsMin=<-15 -15 -15>, angularTarget=<1 0 0>}
      localJointOffset=<4 0 0>
     }]
   }
   n33 = Anim Dynamics - Bone: dyn_fur_arm_r_03(ComponentPose=@n32, Alpha=1, ExternalForce="0,0,0") {
    angularDampingOverride=0.9999
    angularSpringConstant=20000
    bAngularSpring=true
    bGravityOverrideInSimSpace=true
    bOverrideAngularDamping=true
    bUseGravityOverride=true
    gravityScale=0.35
    simulationSpace=RootRelative
    alphaBoolBlend {blendOption=Linear}
    boundBone {boneName=dyn_fur_arm_r_03}
    componentAppliedLinearAccClamp=<3000 3000 3000>
    componentLinearAccScale=<0.4 0.4 0.4>
    componentLinearVelScale=<0.4 0.4 0.4>
    gravityOverride {z=-980}
    physicsBodyDefinitions [{
      sphereCollisionRadius=0
      boundBone {boneName=dyn_fur_arm_r_03}
      boxExtents=<15 18 26>
      constraintSetup {angularLimitsMax=<15 15 15>, angularLimitsMin=<-15 -15 -15>, angularTarget=<1 0 0>}
      localJointOffset=<5.5 0 0>
     }]
   }
   n34 = Anim Dynamics - Bone: dyn_fur_arm_r_04(ComponentPose=@n33, Alpha=1, ExternalForce="0,0,0") {
    angularDampingOverride=0.9999
    angularSpringConstant=20000
    bAngularSpring=true
    bGravityOverrideInSimSpace=true
    bOverrideAngularDamping=true
    bUseGravityOverride=true
    gravityScale=0.35
    simulationSpace=RootRelative
    alphaBoolBlend {blendOption=Linear}
    boundBone {boneName=dyn_fur_arm_r_04}
    componentAppliedLinearAccClamp=<3000 3000 3000>
    componentLinearAccScale=<0.4 0.4 0.4>
    componentLinearVelScale=<0.4 0.4 0.4>
    gravityOverride {z=-980}
    physicsBodyDefinitions [{
      sphereCollisionRadius=0
      boundBone {boneName=dyn_fur_arm_r_04}
      boxExtents=<10 20 20>
      constraintSetup {angularLimitsMax=<15 15 15>, angularLimitsMin=<-15 -15 -15>, angularTarget=<1 0 0>}
      localJointOffset=<3.4 0 0>
     }]
   }
   n35 = Anim Dynamics - Bone: dyn_fur_hand_r_05(ComponentPose=@n34, Alpha=1, ExternalForce="0,0,0") {
    angularDampingOverride=0.9999
    angularSpringConstant=20000
    bAngularSpring=true
    bGravityOverrideInSimSpace=true
    bOverrideAngularDamping=true
    bUseGravityOverride=true
    gravityScale=0.35
    simulationSpace=RootRelative
    alphaBoolBlend {blendOption=Linear}
    boundBone {boneName=dyn_fur_hand_r_05}
    componentAppliedLinearAccClamp=<3000 3000 3000>
    componentLinearAccScale=<0.4 0.4 0.4>
    componentLinearVelScale=<0.4 0.4 0.4>
    gravityOverride {z=-980}
    physicsBodyDefinitions [{
      sphereCollisionRadius=0
      boundBone {boneName=dyn_fur_hand_r_05}
      boxExtents=<8 28 32>
      constraintSetup {angularLimitsMax=<15 15 15>, angularLimitsMin=<-15 -15 -15>, angularTarget=<1 0 0>}
      localJointOffset=<4.5 0 0>
     }]
   }
   n36 = Anim Dynamics - Bone: dyn_fur_hand_r_06(ComponentPose=@n35, Alpha=1, ExternalForce="0,0,0") {
    angularDampingOverride=0.9999
    angularSpringConstant=20000
    bAngularSpring=true
    bGravityOverrideInSimSpace=true
    bOverrideAngularDamping=true
    bUseGravityOverride=true
    gravityScale=0.35
    simulationSpace=RootRelative
    alphaBoolBlend {blendOption=Linear}
    boundBone {boneName=dyn_fur_hand_r_06}
    componentAppliedLinearAccClamp=<3000 3000 3000>
    componentLinearAccScale=<0.4 0.4 0.4>
    componentLinearVelScale=<0.4 0.4 0.4>
    gravityOverride {z=-980}
    physicsBodyDefinitions [{
      sphereCollisionRadius=0
      boundBone {boneName=dyn_fur_hand_r_06}
      boxExtents=<8 28 32>
      constraintSetup {angularLimitsMax=<15 15 15>, angularLimitsMin=<-15 -15 -15>, angularTarget=<1 0 0>}
      localJointOffset=<4 0 0>
     }]
   }
   n37 = Anim Dynamics - Bone: dyn_fur_arm_l_01(ComponentPose=@n36, Alpha=1, ExternalForce="0,0,0") {
    angularDampingOverride=0.9999
    angularSpringConstant=20000
    bAngularSpring=true
    bGravityOverrideInSimSpace=true
    bOverrideAngularDamping=true
    bUseGravityOverride=true
    gravityScale=0.35
    simulationSpace=RootRelative
    alphaBoolBlend {blendOption=Linear}
    boundBone {boneName=dyn_fur_arm_l_01}
    componentAppliedLinearAccClamp=<3000 3000 3000>
    componentLinearAccScale=<0.4 0.4 0.4>
    componentLinearVelScale=<0.4 0.4 0.4>
    gravityOverride {z=-980}
    physicsBodyDefinitions [{
      sphereCollisionRadius=0
      boundBone {boneName=dyn_fur_arm_l_01}
      boxExtents=<10 20 20>
      constraintSetup {angularLimitsMax=<15 15 15>, angularLimitsMin=<-15 -15 -15>, angularTarget=<1 0 0>}
      localJointOffset=<5 0 0>
     }]
   }
   n38 = Anim Dynamics - Bone: dyn_fur_arm_l_02(ComponentPose=@n37, Alpha=1, ExternalForce="0,0,0") {
    angularDampingOverride=0.9999
    angularSpringConstant=20000
    bAngularSpring=true
    bGravityOverrideInSimSpace=true
    bOverrideAngularDamping=true
    bUseGravityOverride=true
    gravityScale=0.35
    simulationSpace=RootRelative
    alphaBoolBlend {blendOption=Linear}
    boundBone {boneName=dyn_fur_arm_l_02}
    componentAppliedLinearAccClamp=<3000 3000 3000>
    componentLinearAccScale=<0.4 0.4 0.4>
    componentLinearVelScale=<0.4 0.4 0.4>
    gravityOverride {z=-980}
    physicsBodyDefinitions [{
      sphereCollisionRadius=0
      boundBone {boneName=dyn_fur_arm_l_02}
      boxExtents=<8 28 32>
      constraintSetup {angularLimitsMax=<15 15 15>, angularLimitsMin=<-15 -15 -15>, angularTarget=<1 0 0>}
      localJointOffset=<4 0 0>
     }]
   }
   n39 = Anim Dynamics - Bone: dyn_fur_arm_l_03(ComponentPose=@n38, Alpha=1, ExternalForce="0,0,0") {
    angularDampingOverride=0.9999
    angularSpringConstant=20000
    bAngularSpring=true
    bGravityOverrideInSimSpace=true
    bOverrideAngularDamping=true
    bUseGravityOverride=true
    gravityScale=0.35
    simulationSpace=RootRelative
    alphaBoolBlend {blendOption=Linear}
    boundBone {boneName=dyn_fur_arm_l_03}
    componentAppliedLinearAccClamp=<3000 3000 3000>
    componentLinearAccScale=<0.4 0.4 0.4>
    componentLinearVelScale=<0.4 0.4 0.4>
    gravityOverride {z=-980}
    physicsBodyDefinitions [{
      sphereCollisionRadius=0
      boundBone {boneName=dyn_fur_arm_l_03}
      boxExtents=<10 20 20>
      constraintSetup {angularLimitsMax=<15 15 15>, angularLimitsMin=<-15 -15 -15>, angularTarget=<1 0 0>}
      localJointOffset=<3.8 0 0>
     }]
   }
   n40 = Anim Dynamics - Bone: dyn_fur_arm_l_04(ComponentPose=@n39, Alpha=1, ExternalForce="0,0,0") {
    angularDampingOverride=0.9999
    angularSpringConstant=20000
    bAngularSpring=true
    bGravityOverrideInSimSpace=true
    bOverrideAngularDamping=true
    bUseGravityOverride=true
    gravityScale=0.35
    simulationSpace=RootRelative
    alphaBoolBlend {blendOption=Linear}
    boundBone {boneName=dyn_fur_arm_l_04}
    componentAppliedLinearAccClamp=<3000 3000 3000>
    componentLinearAccScale=<0.4 0.4 0.4>
    componentLinearVelScale=<0.4 0.4 0.4>
    gravityOverride {z=-980}
    physicsBodyDefinitions [{
      sphereCollisionRadius=0
      boundBone {boneName=dyn_fur_arm_l_04}
      boxExtents=<15 18 26>
      constraintSetup {angularLimitsMax=<15 15 15>, angularLimitsMin=<-15 -15 -15>, angularTarget=<1 0 0>}
      localJointOffset=<4.3 0 0>
     }]
   }
   n41 = Anim Dynamics - Bone: dyn_fur_hand_l_05(ComponentPose=@n40, Alpha=1, ExternalForce="0,0,0") {
    angularDampingOverride=0.9999
    angularSpringConstant=20000
    bAngularSpring=true
    bGravityOverrideInSimSpace=true
    bOverrideAngularDamping=true
    bUseGravityOverride=true
    gravityScale=0.35
    simulationSpace=RootRelative
    alphaBoolBlend {blendOption=Linear}
    boundBone {boneName=dyn_fur_hand_l_05}
    componentAppliedLinearAccClamp=<3000 3000 3000>
    componentLinearAccScale=<0.4 0.4 0.4>
    componentLinearVelScale=<0.4 0.4 0.4>
    gravityOverride {z=-980}
    physicsBodyDefinitions [{
      sphereCollisionRadius=0
      boundBone {boneName=dyn_fur_hand_l_05}
      boxExtents=<10 20 20>
      constraintSetup {angularLimitsMax=<15 15 15>, angularLimitsMin=<-15 -15 -15>, angularTarget=<1 0 0>}
      localJointOffset=<4 0 0>
     }]
   }
   n42 = Anim Dynamics - Bone: dyn_fur_hand_l_06(ComponentPose=@n41, Alpha=1, ExternalForce="0,0,0") {
    angularDampingOverride=0.9999
    angularSpringConstant=20000
    bAngularSpring=true
    bGravityOverrideInSimSpace=true
    bOverrideAngularDamping=true
    bUseGravityOverride=true
    gravityScale=0.35
    simulationSpace=RootRelative
    alphaBoolBlend {blendOption=Linear}
    boundBone {boneName=dyn_fur_hand_l_06}
    componentAppliedLinearAccClamp=<3000 3000 3000>
    componentLinearAccScale=<0.4 0.4 0.4>
    componentLinearVelScale=<0.4 0.4 0.4>
    gravityOverride {z=-980}
    physicsBodyDefinitions [{
      sphereCollisionRadius=0
      boundBone {boneName=dyn_fur_hand_l_06}
      boxExtents=<15 18 26>
      constraintSetup {angularLimitsMax=<15 15 15>, angularLimitsMin=<-15 -15 -15>, angularTarget=<1 0 0>}
      localJointOffset=<5 0 0>
     }]
   }
   n43 = Anim Dynamics - Bone: dyn_fur_torso_r_01(ComponentPose=@n42, Alpha=1, ExternalForce="0,0,0") {
    angularDampingOverride=0.9999
    angularSpringConstant=20000
    bAngularSpring=true
    bGravityOverrideInSimSpace=true
    bOverrideAngularDamping=true
    bUseGravityOverride=true
    gravityScale=0.35
    simulationSpace=RootRelative
    alphaBoolBlend {blendOption=Linear}
    boundBone {boneName=dyn_fur_torso_r_01}
    componentAppliedLinearAccClamp=<3000 3000 3000>
    componentLinearAccScale=<0.4 0.4 0.4>
    componentLinearVelScale=<0.4 0.4 0.4>
    gravityOverride {z=-980}
    physicsBodyDefinitions [{
      sphereCollisionRadius=0
      boundBone {boneName=dyn_fur_torso_r_01}
      boxExtents=<10 20 20>
      constraintSetup {angularLimitsMax=<15 15 15>, angularLimitsMin=<-15 -15 -15>, angularTarget=<1 0 0>}
      localJointOffset=<5 0 0>
     }]
   }
   n44 = Anim Dynamics - Bone: dyn_fur_torso_r_02(ComponentPose=@n43, Alpha=1, ExternalForce="0,0,0") {
    angularDampingOverride=0.9999
    angularSpringConstant=20000
    bAngularSpring=true
    bGravityOverrideInSimSpace=true
    bOverrideAngularDamping=true
    bUseGravityOverride=true
    gravityScale=0.35
    simulationSpace=RootRelative
    alphaBoolBlend {blendOption=Linear}
    boundBone {boneName=dyn_fur_torso_r_02}
    componentAppliedLinearAccClamp=<3000 3000 3000>
    componentLinearAccScale=<0.4 0.4 0.4>
    componentLinearVelScale=<0.4 0.4 0.4>
    gravityOverride {z=-980}
    physicsBodyDefinitions [{
      sphereCollisionRadius=0
      boundBone {boneName=dyn_fur_torso_r_02}
      boxExtents=<8 28 32>
      constraintSetup {angularLimitsMax=<15 15 15>, angularLimitsMin=<-15 -15 -15>, angularTarget=<1 0 0>}
      localJointOffset=<4 0 0>
     }]
   }
   n45 = Anim Dynamics - Bone: dyn_fur_torso_l_01(ComponentPose=@n44, Alpha=1, ExternalForce="0,0,0") {
    angularDampingOverride=0.9999
    angularSpringConstant=20000
    bAngularSpring=true
    bGravityOverrideInSimSpace=true
    bOverrideAngularDamping=true
    bUseGravityOverride=true
    gravityScale=0.35
    simulationSpace=RootRelative
    alphaBoolBlend {blendOption=Linear}
    boundBone {boneName=dyn_fur_torso_l_01}
    componentAppliedLinearAccClamp=<3000 3000 3000>
    componentLinearAccScale=<0.4 0.4 0.4>
    componentLinearVelScale=<0.4 0.4 0.4>
    gravityOverride {z=-980}
    physicsBodyDefinitions [{
      sphereCollisionRadius=0
      boundBone {boneName=dyn_fur_torso_l_01}
      boxExtents=<10 20 20>
      constraintSetup {angularLimitsMax=<15 15 15>, angularLimitsMin=<-15 -15 -15>, angularTarget=<1 0 0>}
      localJointOffset=<5 0 0>
     }]
   }
   n46 = Anim Dynamics - Bone: dyn_fur_torso_l_02(ComponentPose=@n45, Alpha=1, ExternalForce="0,0,0") {
    angularDampingOverride=0.9999
    angularSpringConstant=20000
    bAngularSpring=true
    bGravityOverrideInSimSpace=true
    bOverrideAngularDamping=true
    bUseGravityOverride=true
    gravityScale=0.35
    simulationSpace=RootRelative
    alphaBoolBlend {blendOption=Linear}
    boundBone {boneName=dyn_fur_torso_l_02}
    componentAppliedLinearAccClamp=<3000 3000 3000>
    componentLinearAccScale=<0.4 0.4 0.4>
    componentLinearVelScale=<0.4 0.4 0.4>
    gravityOverride {z=-980}
    physicsBodyDefinitions [{
      sphereCollisionRadius=0
      boundBone {boneName=dyn_fur_torso_l_02}
      boxExtents=<8 28 32>
      constraintSetup {angularLimitsMax=<15 15 15>, angularLimitsMin=<-15 -15 -15>, angularTarget=<1 0 0>}
      localJointOffset=<4 0 0>
     }]
   }
   n47 = Anim Dynamics - Bone: dyn_fur_leg_r_01(ComponentPose=@n46, Alpha=1, ExternalForce="0,0,0") {
    angularDampingOverride=0.9999
    angularSpringConstant=20000
    bAngularSpring=true
    bGravityOverrideInSimSpace=true
    bOverrideAngularDamping=true
    bUseGravityOverride=true
    gravityScale=0.35
    simulationSpace=RootRelative
    alphaBoolBlend {blendOption=Linear}
    boundBone {boneName=dyn_fur_leg_r_01}
    componentAppliedLinearAccClamp=<3000 3000 3000>
    componentLinearAccScale=<0.4 0.4 0.4>
    componentLinearVelScale=<0.4 0.4 0.4>
    gravityOverride {z=-980}
    physicsBodyDefinitions [{
      sphereCollisionRadius=0
      boundBone {boneName=dyn_fur_leg_r_01}
      boxExtents=<15 18 26>
      constraintSetup {angularLimitsMax=<15 15 15>, angularLimitsMin=<-15 -15 -15>, angularTarget=<1 0 0>}
      localJointOffset=<5 0 0>
     }]
   }
   n48 = Anim Dynamics - Bone: dyn_fur_leg_r_02(ComponentPose=@n47, Alpha=1, ExternalForce="0,0,0") {
    angularDampingOverride=0.9999
    angularSpringConstant=20000
    bAngularSpring=true
    bGravityOverrideInSimSpace=true
    bOverrideAngularDamping=true
    bUseGravityOverride=true
    gravityScale=0.35
    simulationSpace=RootRelative
    alphaBoolBlend {blendOption=Linear}
    boundBone {boneName=dyn_fur_leg_r_02}
    componentAppliedLinearAccClamp=<3000 3000 3000>
    componentLinearAccScale=<0.4 0.4 0.4>
    componentLinearVelScale=<0.4 0.4 0.4>
    gravityOverride {z=-980}
    physicsBodyDefinitions [{
      sphereCollisionRadius=0
      boundBone {boneName=dyn_fur_leg_r_02}
      boxExtents=<10 20 20>
      constraintSetup {angularLimitsMax=<15 15 15>, angularLimitsMin=<-15 -15 -15>, angularTarget=<1 0 0>}
      localJointOffset=<4 0 0>
     }]
   }
   n49 = Anim Dynamics - Bone: dyn_fur_leg_r_03(ComponentPose=@n48, Alpha=1, ExternalForce="0,0,0") {
    angularDampingOverride=0.9999
    angularSpringConstant=20000
    bAngularSpring=true
    bGravityOverrideInSimSpace=true
    bOverrideAngularDamping=true
    bUseGravityOverride=true
    gravityScale=0.35
    simulationSpace=RootRelative
    alphaBoolBlend {blendOption=Linear}
    boundBone {boneName=dyn_fur_leg_r_03}
    componentAppliedLinearAccClamp=<3000 3000 3000>
    componentLinearAccScale=<0.4 0.4 0.4>
    componentLinearVelScale=<0.4 0.4 0.4>
    gravityOverride {z=-980}
    physicsBodyDefinitions [{
      sphereCollisionRadius=0
      boundBone {boneName=dyn_fur_leg_r_03}
      boxExtents=<10 20 20>
      constraintSetup {angularLimitsMax=<15 15 15>, angularLimitsMin=<-15 -15 -15>, angularTarget=<1 0 0>}
      localJointOffset=<5 0 0>
     }]
   }
   n50 = Anim Dynamics - Bone: dyn_fur_leg_r_04(ComponentPose=@n49, Alpha=1, ExternalForce="0,0,0") {
    angularDampingOverride=0.9999
    angularSpringConstant=20000
    bAngularSpring=true
    bGravityOverrideInSimSpace=true
    bOverrideAngularDamping=true
    bUseGravityOverride=true
    gravityScale=0.35
    simulationSpace=RootRelative
    alphaBoolBlend {blendOption=Linear}
    boundBone {boneName=dyn_fur_leg_r_04}
    componentAppliedLinearAccClamp=<3000 3000 3000>
    componentLinearAccScale=<0.4 0.4 0.4>
    componentLinearVelScale=<0.4 0.4 0.4>
    gravityOverride {z=-980}
    physicsBodyDefinitions [{
      sphereCollisionRadius=0
      boundBone {boneName=dyn_fur_leg_r_04}
      boxExtents=<8 28 32>
      constraintSetup {angularLimitsMax=<15 15 15>, angularLimitsMin=<-15 -15 -15>, angularTarget=<1 0 0>}
      localJointOffset=<3.6 0 0>
     }]
   }
   n51 = Anim Dynamics - Bone: dyn_fur_leg_r_05(ComponentPose=@n50, Alpha=1, ExternalForce="0,0,0") {
    angularDampingOverride=0.9999
    angularSpringConstant=20000
    bAngularSpring=true
    bGravityOverrideInSimSpace=true
    bOverrideAngularDamping=true
    bUseGravityOverride=true
    gravityScale=0.35
    simulationSpace=RootRelative
    alphaBoolBlend {blendOption=Linear}
    boundBone {boneName=dyn_fur_leg_r_05}
    componentAppliedLinearAccClamp=<3000 3000 3000>
    componentLinearAccScale=<0.4 0.4 0.4>
    componentLinearVelScale=<0.4 0.4 0.4>
    gravityOverride {z=-980}
    physicsBodyDefinitions [{
      sphereCollisionRadius=0
      boundBone {boneName=dyn_fur_leg_r_05}
      boxExtents=<15 18 26>
      constraintSetup {angularLimitsMax=<15 15 15>, angularLimitsMin=<-15 -15 -15>, angularTarget=<1 0 0>}
      localJointOffset=<4 0 0>
     }]
   }
   n52 = Anim Dynamics - Bone: dyn_fur_leg_l_01(ComponentPose=@n51, Alpha=1, ExternalForce="0,0,0") {
    angularDampingOverride=0.9999
    angularSpringConstant=20000
    bAngularSpring=true
    bGravityOverrideInSimSpace=true
    bOverrideAngularDamping=true
    bUseGravityOverride=true
    gravityScale=0.35
    simulationSpace=RootRelative
    alphaBoolBlend {blendOption=Linear}
    boundBone {boneName=dyn_fur_leg_l_01}
    componentAppliedLinearAccClamp=<3000 3000 3000>
    componentLinearAccScale=<0.4 0.4 0.4>
    componentLinearVelScale=<0.4 0.4 0.4>
    gravityOverride {z=-980}
    physicsBodyDefinitions [{
      sphereCollisionRadius=0
      boundBone {boneName=dyn_fur_leg_l_01}
      boxExtents=<15 18 26>
      constraintSetup {angularLimitsMax=<15 15 15>, angularLimitsMin=<-15 -15 -15>, angularTarget=<1 0 0>}
      localJointOffset=<4 0 0>
     }]
   }
   n53 = Anim Dynamics - Bone: dyn_fur_leg_l_02(ComponentPose=@n52, Alpha=1, ExternalForce="0,0,0") {
    angularDampingOverride=0.9999
    angularSpringConstant=20000
    bAngularSpring=true
    bGravityOverrideInSimSpace=true
    bOverrideAngularDamping=true
    bUseGravityOverride=true
    gravityScale=0.35
    simulationSpace=RootRelative
    alphaBoolBlend {blendOption=Linear}
    boundBone {boneName=dyn_fur_leg_l_02}
    componentAppliedLinearAccClamp=<3000 3000 3000>
    componentLinearAccScale=<0.4 0.4 0.4>
    componentLinearVelScale=<0.4 0.4 0.4>
    gravityOverride {z=-980}
    physicsBodyDefinitions [{
      sphereCollisionRadius=0
      boundBone {boneName=dyn_fur_leg_l_02}
      boxExtents=<15 18 26>
      constraintSetup {angularLimitsMax=<15 15 15>, angularLimitsMin=<-15 -15 -15>, angularTarget=<1 0 0>}
      localJointOffset=<5 0 0>
     }]
   }
   n54 = Anim Dynamics - Bone: dyn_fur_leg_l_03(ComponentPose=@n53, Alpha=1, ExternalForce="0,0,0") {
    angularDampingOverride=0.9999
    angularSpringConstant=20000
    bAngularSpring=true
    bGravityOverrideInSimSpace=true
    bOverrideAngularDamping=true
    bUseGravityOverride=true
    gravityScale=0.35
    simulationSpace=RootRelative
    alphaBoolBlend {blendOption=Linear}
    boundBone {boneName=dyn_fur_leg_l_03}
    componentAppliedLinearAccClamp=<3000 3000 3000>
    componentLinearAccScale=<0.4 0.4 0.4>
    componentLinearVelScale=<0.4 0.4 0.4>
    gravityOverride {z=-980}
    physicsBodyDefinitions [{
      sphereCollisionRadius=0
      boundBone {boneName=dyn_fur_leg_l_03}
      boxExtents=<8 28 32>
      constraintSetup {angularLimitsMax=<15 15 15>, angularLimitsMin=<-15 -15 -15>, angularTarget=<1 0 0>}
      localJointOffset=<3.6 0 0>
     }]
   }
   n55 = Anim Dynamics - Bone: dyn_fur_leg_l_04(ComponentPose=@n54, Alpha=1, ExternalForce="0,0,0") {
    angularDampingOverride=0.9999
    angularSpringConstant=20000
    bAngularSpring=true
    bGravityOverrideInSimSpace=true
    bOverrideAngularDamping=true
    bUseGravityOverride=true
    gravityScale=0.35
    simulationSpace=RootRelative
    alphaBoolBlend {blendOption=Linear}
    boundBone {boneName=dyn_fur_leg_l_04}
    componentAppliedLinearAccClamp=<3000 3000 3000>
    componentLinearAccScale=<0.4 0.4 0.4>
    componentLinearVelScale=<0.4 0.4 0.4>
    gravityOverride {z=-980}
    physicsBodyDefinitions [{
      sphereCollisionRadius=0
      boundBone {boneName=dyn_fur_leg_l_04}
      boxExtents=<10 20 20>
      constraintSetup {angularLimitsMax=<15 15 15>, angularLimitsMin=<-15 -15 -15>, angularTarget=<1 0 0>}
      localJointOffset=<5 0 0>
     }]
   }
   n56 = Anim Dynamics - Bone: dyn_fur_leg_l_05(ComponentPose=@n55, Alpha=1, ExternalForce="0,0,0") {
    angularDampingOverride=0.9999
    angularSpringConstant=20000
    bAngularSpring=true
    bGravityOverrideInSimSpace=true
    bOverrideAngularDamping=true
    bUseGravityOverride=true
    gravityScale=0.35
    simulationSpace=RootRelative
    alphaBoolBlend {blendOption=Linear}
    boundBone {boneName=dyn_fur_leg_l_05}
    componentAppliedLinearAccClamp=<3000 3000 3000>
    componentLinearAccScale=<0.4 0.4 0.4>
    componentLinearVelScale=<0.4 0.4 0.4>
    gravityOverride {z=-980}
    physicsBodyDefinitions [{
      sphereCollisionRadius=0
      boundBone {boneName=dyn_fur_leg_l_05}
      boxExtents=<8 28 32>
      constraintSetup {angularLimitsMax=<15 15 15>, angularLimitsMin=<-15 -15 -15>, angularTarget=<1 0 0>}
      localJointOffset=<3.6 0 0>
     }]
   }
   Output Pose(Result=Component To Local(ComponentPose=@n56)) {layerGroup=None}
  ]}
 EventGraph {nodes [
   Event OnCharacterCustomizationLoaded()
   > n1 = Cast To FortPlayerPawn(Object=TryGetFortPawn())
   > Set HeadMesh(HeadMesh:Skeletal Mesh Component Object Reference=GetSkeletalMeshForPartType(self=@n1, PartType=Head)) -> (Output_Get:Skeletal Mesh Component Object Reference)
  ]}
]
variables [HeadMesh {type=SkeletalMeshComponent}]
